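# M_Test is a test material in the GorgeDust grass foliage folder, set to two-sided Default Lit shading.
# It holds a SplineThicken function call with one constant driving both base and tip width at a very large value.
# No material output connections appear in the graph, so the function's result is not shown feeding the surface or vertex offset.
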Material /WildEstate/Foliage/Grass/GorgeDust/M_Test
graph {
 properties {bAutomaticallySetUsageInEditor=false, twoSided=true, usedShadingModels=Default Lit}
 nodes [
  n1 = Constant() {r=1000}
  MaterialFunctionCall(WidthBase=@n1, WidthTip=@n1) {materialFunction=&SplineThicken}
 ]
}
